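; Giant tree petals actor for the Artemis Daisy set that populates its DaisyPetals_HISM components.
; Construction Script only messages a valid editor actor component through DaisyPetalInterface to request an editor-only reconstruction.
; handleReconstruction, with no caller in this graph, optionally randomizes the seed and calls GenerateMeshes.
; GenerateMeshes passes that seed to components set to inherit it, then runs PopulatePetals on each.
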(event EventBeginPlay)

(event Collision|EventActorBeginOverlap
  (params (OtherActor :type Actor%20Object%20Reference)))

(event EventTick
  (params (DeltaSeconds :type Float%20%28single-precision%29)))

(fn GenerateMeshes
  (for _array_element (Actor|GetComponentsByClass "/Game/Environments/Artemis/Sets/Daisy/GiantTree_Petals/DaisyPetals_HISM.DaisyPetals_HISM_C")
    (Utilities|FlowControl|Sequence
      (:then_0
        (if (Variables|/Game/Environments/Artemis/Sets/Daisy/GiantTree_Petals/DaisyPetals_HISM.DaisyPetals_HISM_C|GetInheritRandomSeedFromParentActor _array_element)
          (Variables|/Game/Environments/Artemis/Sets/Daisy/GiantTree_Petals/DaisyPetals_HISM.DaisyPetals_HISM_C|SetRandomSeed (Variables|Default|GetRandomSeed) _array_element)))
      (:then_1
        (Class|DaisyPetalsHISM|PopulatePetals _array_element)))))

(fn handleReconstruction
  (Utilities|FlowControl|Sequence
    (:then_0
      (if (Variables|Default|GetRandomizeSeedOnConstruct)
        (Variables|Default|SetRandomSeed (Math|Random|RandomIntegerinRange 0 999999))))
    (:then_1
      (if (Variables|Default|GetGenerateMeshesOnConstruct)
        (CallFunction|GenerateMeshes)))))

(fn ConstructionScript
  (bind _daisypetals_editoractorcomp (Variables|Default|GetDaisyPetals_EditorActorComp))
  (Utilities|IsValid _daisypetals_editoractorcomp
    (:Is%20Valid
      (Class|DaisyPetalInterface|EdiotrOnlyReconstruct(Message) _daisypetals_editoractorcomp))))
