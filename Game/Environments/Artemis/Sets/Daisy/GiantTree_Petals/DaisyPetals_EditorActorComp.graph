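# Editor-only component blueprint that defers reconstruction of the Daisy set's BP_Artemis_GiantTree_Petals actor.
# EdiotrOnly_Reconstruct starts a short one-shot timer only when construct_timerHandle is not already valid, so repeated requests coalesce into one.
# When the timer fires, it clears the handle, casts the owner to the petals blueprint and calls handleReconstruction on it.
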EditorUtilityBlueprint /Game/Environments/Artemis/Sets/Daisy/GiantTree_Petals/DaisyPetals_EditorActorComp
graphs [EventGraph {nodes [
   n1 = CustomEvent() {is=title K2Node_CustomEvent}
   > Clear and Invalidate Timer by Handle!(execute, WorldContextObject:Object Reference, Handle=Get construct_timerHandle() -> (construct_timerHandle:Timer Handle Structure)) -> (then)
   > n2 = Cast To BP_Artemis_GiantTree_Petals!(execute, Object=GetOwner()) -> (then, CastFailed, AsBP Artemis Giant Tree Petals:BP Artemis Giant Tree Petals Object Reference) {is=title K2Node_DynamicCast}
   > handleReconstruction(self:BP Artemis Giant Tree Petals Object Reference=@n2)
   Event EdiotrOnly_Reconstruct() {is=title K2Node_Event}
   > n3 = Branch(Condition=Is Valid Timer Handle!(Handle:Timer Handle Structure=Get construct_timerHandle()) -> (ReturnValue:bool))
   n4 = Set Timer by Event(execute=@n3.else, Delegate=@n1, Time=0.1, bLooping=false)
   > Set construct_timerHandle(construct_timerHandle:Timer Handle Structure=@n4) -> (Output_Get:Timer Handle Structure)
   Event Begin Play()
   Event Tick!() -> (OutputDelegate, then, DeltaSeconds:float)
  ]}]
variables [construct_timerHandle {type=TimerHandle}]
