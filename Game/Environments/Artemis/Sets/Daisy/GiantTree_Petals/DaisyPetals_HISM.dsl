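; Blueprint that builds a ring of daisy petals as instanced static meshes for the Artemis giant tree set.
; PopulatePetals clears the instances, then AddPetals spaces the petal count evenly around 360 degrees of yaw and adds all transforms at once.
; Validate functions supply each petal's center offset, rotation and scale, with optional randomness.
; Random streams combine the base seed with a square-root hash of the index, keeping variation repeatable per petal.
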(event EventBeginPlay)

(event EventTick
  (params (DeltaSeconds :type Float%20%28single-precision%29)))

(fn AddPetals
  (params (count :type Integer))
  (LocalVariables|Setl_count count)
  (LocalVariables|Setl_seed (Variables|Petals|GetRandomSeed))
  (LocalVariables|Setl_rotPerIndex (/ 3.6e+02 (Math|Conversions|ToFloat(Integer) (LocalVariables|Getl_count))))
  (Utilities|FlowControl|Sequence
    (:then_0
      (for _index (range (LocalVariables|Getl_count))
        (LocalVariables|Setl_curIndex _index)
        (Utilities|FlowControl|Sequence
          (:then_0
            (bind _output_get (Variables|Setl_rot_Roll (Math|Rotator|MakeRotator 0.0 0.0 (* (LocalVariables|Getl_rotPerIndex) (LocalVariables|Getl_curIndex)))))
            (bind _output_get_2 (LocalVariables|Setl_directionRot _output_get))
            (bind _returnvalue (CallFunction|ValidateRotation _output_get_2 (LocalVariables|Getl_curIndex) (LocalVariables|Getl_seed)))
            (LocalVariables|Setl_rot _returnvalue))
          (:then_1
            (bind _offset (CallFunction|ValidateOffsetfromCenter (LocalVariables|Getl_curIndex) (LocalVariables|Getl_seed)))
            (LocalVariables|Setl_offsetFromCenter _offset))
          (:then_2
            (bind _returnvalue_2 (CallFunction|ValidateScale (LocalVariables|Getl_curIndex) (LocalVariables|Getl_seed)))
            (LocalVariables|Setl_scale _returnvalue_2))
          (:then_3
            (Utilities|Array|Add (LocalVariables|Getl_transforms) (Math|Transform|MakeTransform (* (Math|Vector|GetRightVector (LocalVariables|Getl_directionRot)) (LocalVariables|Getl_offsetFromCenter)) (LocalVariables|Getl_rot) (LocalVariables|Getl_scale)))))))
    (:then_1
      (Components|InstancedStaticMesh|AddInstances (LocalVariables|Getl_transforms)))))

(fn convertIndexToSeedOffset
  (params (InInt :type Integer))
  (return (Math|Float|Round (* (Math|Float|Fraction (Math|Float|Sqrt (Math|Conversions|ToFloat(Integer) InInt))) 1.5e+03))))

(fn PopulatePetals
  (Components|InstancedStaticMesh|ClearInstances)
  (CallFunction|AddPetals (Variables|Petals|GetPetalCount)))

(fn ValidateOffsetFromCenter
  (params (instanceIndex :type Integer) (seed :type Integer))
  (LocalVariables|Setl_seed (Math|Random|MakeRandomStream (+ seed (CallFunction|ConvertIndextoSeedOffset instanceIndex))))
  (if (Variables|Petals|PetalPosition|GetUseRandomOffsetFromCenter)
    (return (Math|Random|RandomFloatinRangefromStream (LocalVariables|Getl_seed) (Variables|Petals|PetalPosition|GetPetalOffsetFromCenter) (Variables|Petals|PetalPosition|GetPetalOffsetFromCenter_RandomMax)))
    (else
      (return (Variables|Petals|PetalPosition|GetPetalOffsetFromCenter)))))

(fn ValidateRotation
  (params (InRot :type Rotator) (instanceIndex :type Integer) (seed :type Integer))
  (LocalVariables|Setl_inRot InRot)
  (LocalVariables|Setl_seed (Math|Random|MakeRandomStream (+ seed (CallFunction|ConvertIndextoSeedOffset instanceIndex))))
  (Utilities|FlowControl|Sequence
    (:then_0
      (if (Variables|Petals|PetalRotation|GetUseRandomRotation)
        (Variables|Setl_finalRotOffset_Roll (Math|Rotator|MakeRotator (Math|Random|RandomFloatinRangefromStream (LocalVariables|Getl_seed) (.roll (Variables|Petals|PetalRotation|GetPetalRotationOffset)) (.roll (Variables|Petals|PetalRotation|GetPetalRotationOffset_RandomMax))) (Math|Random|RandomFloatinRangefromStream (LocalVariables|Getl_seed) (.pitch (Variables|Petals|PetalRotation|GetPetalRotationOffset)) (.pitch (Variables|Petals|PetalRotation|GetPetalRotationOffset_RandomMax))) (Math|Random|RandomFloatinRangefromStream (LocalVariables|Getl_seed) (.yaw (Variables|Petals|PetalRotation|GetPetalRotationOffset)) (.yaw (Variables|Petals|PetalRotation|GetPetalRotationOffset_RandomMax)))))
        (else
          (LocalVariables|Setl_finalRotOffset (Variables|Petals|PetalRotation|GetPetalRotationOffset)))))
    (:then_1
      (LocalVariables|Setl_finalRotOffset (Math|Rotator|MakeRotator (- (.roll (LocalVariables|Getl_inRot)) (.pitch (LocalVariables|Getl_finalRotOffset))) (+ (.pitch (LocalVariables|Getl_inRot)) (.roll (LocalVariables|Getl_finalRotOffset))) (+ (.yaw (LocalVariables|Getl_inRot)) (.yaw (LocalVariables|Getl_finalRotOffset))))))
    (:then_2
      (return (LocalVariables|Getl_finalRotOffset)))))

(fn ValidateScale
  (params (instanceIndex :type Integer) (seed :type Integer))
  (LocalVariables|Setl_scale (Math|Conversions|ToVector(Float) (Variables|Petals|PetalScale|GetPetalGlobalScale)))
  (LocalVariables|Setl_seed (Math|Random|MakeRandomStream (+ seed (CallFunction|ConvertIndextoSeedOffset instanceIndex))))
  (Utilities|FlowControl|Sequence
    (:then_0
      (if (and (Variables|Petals|PetalScale|GetUseRandomScale) (Utilities|Operators|NotEqual(!=) (Variables|Petals|PetalScale|GetPetalRandomScaleRange) "1.000000,1.000000,1.000000" 0))
        (LocalVariables|Setl_scale (Math|Vector|MakeVector (Math|Random|RandomFloatinRangefromStream (LocalVariables|Getl_seed) (.x (- (Variables|Petals|PetalScale|GetPetalGlobalScale) (- (* (Variables|Petals|PetalScale|GetPetalGlobalScale) (Variables|Petals|PetalScale|GetPetalRandomScaleRange)) (Variables|Petals|PetalScale|GetPetalGlobalScale)))) (.x (* (Variables|Petals|PetalScale|GetPetalGlobalScale) (Variables|Petals|PetalScale|GetPetalRandomScaleRange)))) (Math|Random|RandomFloatinRangefromStream (LocalVariables|Getl_seed) (.y (- (Variables|Petals|PetalScale|GetPetalGlobalScale) (- (* (Variables|Petals|PetalScale|GetPetalGlobalScale) (Variables|Petals|PetalScale|GetPetalRandomScaleRange)) (Variables|Petals|PetalScale|GetPetalGlobalScale)))) (.y (* (Variables|Petals|PetalScale|GetPetalGlobalScale) (Variables|Petals|PetalScale|GetPetalRandomScaleRange)))) (Math|Random|RandomFloatinRangefromStream (LocalVariables|Getl_seed) (.z (- (Variables|Petals|PetalScale|GetPetalGlobalScale) (- (* (Variables|Petals|PetalScale|GetPetalGlobalScale) (Variables|Petals|PetalScale|GetPetalRandomScaleRange)) (Variables|Petals|PetalScale|GetPetalGlobalScale)))) (.z (* (Variables|Petals|PetalScale|GetPetalGlobalScale) (Variables|Petals|PetalScale|GetPetalRandomScaleRange))))))))
    (:then_1
      (return (LocalVariables|Getl_scale)))))
